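; Environment prop Blueprint that scatters decorative add-on meshes on the sockets of its static mesh.
; The construction script seeds a random stream from actor location, filters sockets by name, shuffles them and keeps a set or random percentage.
; Chosen sockets receive randomly rotated instances on one added HISM component, with a random material and optional HSV-packed tint via custom primitive data.
; RandomizeAll rerolls the layout, with a DELETEME-marked double-spawn path for shoe racks.
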(fn AddNewInstanceToComponent
  (params (comp :type Instanced%20Static%20Mesh%20Component%20Object%20Reference) (InTransform :type Transform%20%28by%20ref%29))
  (bind _location (.location InTransform))
  (bind _addonmeshscale (Variables|HiddenProperties|GetAddonMeshScale))
  (bind _returnvalue (Math|Vector|MakeVector _addonmeshscale _addonmeshscale _addonmeshscale))
  (if (Utilities|Array|IsNotEmpty (Variables|AddonDeco|GetMeshList))
    (if (Variables|AddonDeco|GetUseRandomRotation)
      (Components|InstancedStaticMesh|AddInstance comp (Math|Transform|MakeTransform _location (Math|Random|RandomRotatorfromStream (Variables|HiddenProperties|GetRandomStream)) _returnvalue) false)
      (else
        (Components|InstancedStaticMesh|AddInstance comp (Math|Transform|MakeTransform _location (Math|Rotator|CombineRotators (.rotation InTransform) (Math|Rotator|MakeRotator 0.0 0.0 (Math|Random|RandomFloatinRangefromStream (Variables|HiddenProperties|GetRandomStream) (- (Variables|AddonDeco|GetRandomRotationAmount) (* 2.0 (Variables|AddonDeco|GetRandomRotationAmount))) (Variables|AddonDeco|GetRandomRotationAmount)))) _returnvalue) false)))))

(fn DELETEME_DoubleSpawnDeco
  (params (InSocketName :type Name))
  (bind _returnvalue_location (Math|Transform|BreakTransform (Transformation|GetSocketTransform (Variables|Debug|GetStaticMeshComponent) InSocketName "RTS_Actor")))
  (bind _returnvalue_location_2 (Math|Transform|BreakTransform (Transformation|GetSocketTransform (Variables|Debug|GetStaticMeshComponent) InSocketName "RTS_Actor")))
  (bind _returnvalue (AddComponent|NODE_AddHierarchicalInstancedStaticMeshComponent-3))
  (Components|StaticMesh|SetStaticMesh _returnvalue (Variables|HiddenProperties|GetInstancedMeshToSpawn))
  (Utilities|Array|Add (Variables|HiddenProperties|GetAR_InstancedMeshComponents) _returnvalue)
  (Components|InstancedStaticMesh|AddInstance _returnvalue (Math|Transform|MakeTransform (Math|Vector|MakeVector (- (.x _returnvalue_location_2) (Variables|AddonDeco|GetOffsetDistance)) (.y _returnvalue_location_2) (.z _returnvalue_location_2)) (Math|Rotator|MakeRotator 0.0 0.0 (* (Math|Random|RandomFloatinRange (- (Variables|AddonDeco|GetOffsetRotationRange) (* (Variables|AddonDeco|GetOffsetRotationRange) 2.0)) (Variables|AddonDeco|GetOffsetRotationRange)) 0.35)) "-1.000000,1.000000,1.000000") false)
  (Utilities|Array|Add (Variables|HiddenProperties|GetAR_InstancedMeshComponents) _returnvalue)
  (Components|InstancedStaticMesh|AddInstance _returnvalue (Math|Transform|MakeTransform (Math|Vector|MakeVector (+ (.x _returnvalue_location) (Variables|AddonDeco|GetOffsetDistance)) (.y _returnvalue_location) (.z _returnvalue_location)) (Math|Rotator|MakeRotator 0.0 0.0 (Math|Random|RandomFloatinRange (- (Variables|AddonDeco|GetOffsetRotationRange) (* (Variables|AddonDeco|GetOffsetRotationRange) 2.0)) (Variables|AddonDeco|GetOffsetRotationRange)))) false))

(fn GetSocketAmountFromPercentage
  (Variables|HiddenProperties|SetSpawnSocketAmount (/ (* (Variables|HiddenProperties|GetMaxSocketAmount) (Variables|AddonDeco|GetSpawnPercentage)) 100)))

(fn GetValidSockets
  (bind _output_get (LocalVariables|SetL_AR_AllSockets (Transformation|GetAllSocketNames (Variables|Debug|GetStaticMeshComponent))))
  (for _array_element _output_get
    (if (Utilities|String|Contains (Utilities|String|ToString(Name) _array_element) (Variables|HiddenProperties|GetSocketName))
      (Utilities|Array|AddUnique (Variables|HiddenProperties|GetAR_ValidSocketList) _array_element))))

(fn InitializeRandomStream
  (Variables|HiddenProperties|SetRandomStream (Math|Random|MakeRandomStreamfromLocation (Transformation|GetActorLocation) 1.0 true)))

(fn RandomizeSpawnPercentage
  (if (Variables|AddonDeco|GetRandomizeSpawnPercentage)
    (Variables|AddonDeco|SetSpawnPercentage (Math|Random|RandomIntegerinRangefromStream (Variables|HiddenProperties|GetRandomStream) 0 100))))

(fn RandomizeAll
  (AddonDeco|RemoveAddonMeshes)
  (AddonDeco|GetValidSockets)
  (AddonDeco|RandomizeSpawnPercentage)
  (AddonDeco|SetMaxSocketAmount)
  (AddonDeco|GetSocketAmountfromPercentage)
  (AddonDeco|RandomizeSpawnSockets)
  (for _array_element (Variables|HiddenProperties|GetAR_RandomizedSockets)
    (if (Variables|AddonDeco|GetIsShoeRack)
      (AddonDeco|DELETEMEDoubleSpawnDeco _array_element)
      (else
        (AddonDeco|AddNewInstancetoComponent 0 (Transformation|GetSocketTransform (Variables|Debug|GetStaticMeshComponent) _array_element "RTS_Actor"))
        (AddonDeco|RandomizeMaterials)))))

(fn RandomizeMaterials
  (bind _tintcolor (Variables|AddonDeco|GetTintColor))
  (bind _returnvalue (Math|Color|RGBtoHSV(LinearColor) _tintcolor))
  (bind (_r _g _b) (Utilities|Struct|BreakLinearColor _returnvalue))
  (bind _returnvalue_2 (Math|Color|Fortnite|PackHSVtoFloat _r _g _b))
  (bind _ar_materiallist (Variables|HiddenProperties|GetAR_MaterialList))
  (for _array_element (Variables|HiddenProperties|GetAR_InstancedMeshComponents)
    (if (> (Utilities|Array|Length _ar_materiallist) 0)
      (Rendering|Material|SetMaterial _array_element 0 (Utilities|Array|RandomArrayItemfromStream _ar_materiallist (Variables|HiddenProperties|GetRandomStream)))
      (else
        _))
    (if (Variables|AddonDeco|GetEnableTintColor)
      (Rendering|Material|SetCustomPrimitiveDataFloat _array_element 30 _returnvalue_2)
      (Rendering|Material|SetDefaultCustomPrimitiveDataFloat _array_element 30 _returnvalue_2))))

(fn RandomizeSpawnSockets
  (bind _ar_validsocketlist (Variables|HiddenProperties|GetAR_ValidSocketList))
  (if (!= (Variables|HiddenProperties|GetSpawnSocketAmount) 0)
    (Utilities|Array|Clear (Variables|HiddenProperties|GetAR_RandomizedSockets))
    (Utilities|Array|ShufflefromStream _ar_validsocketlist (Variables|HiddenProperties|GetRandomStream))
    (for _index (range (Variables|HiddenProperties|GetSpawnSocketAmount))
      (Utilities|Array|AddUnique (Variables|HiddenProperties|GetAR_RandomizedSockets) (Utilities|Array|Get(acopy) _ar_validsocketlist _index)))))

(fn RemoveAddonMeshes
  (for _array_element (Variables|HiddenProperties|GetAR_InstancedMeshComponents)
    (Components|DestroyComponent _array_element)))

(fn SetMaxSocketAmount
  (Variables|HiddenProperties|SetMaxSocketAmount (Utilities|Array|Length (Variables|HiddenProperties|GetAR_ValidSocketList))))

(fn ConstructionScript
  (bind _returnvalue (Actor|GetComponentbyClass "/Script/Engine.HierarchicalInstancedStaticMeshComponent"))
  (Utilities|FlowControl|Sequence
    (:then_0
      (AddonDeco|InitializeRandomStream)
      (AddonDeco|GetValidSockets)
      (AddonDeco|RandomizeSpawnPercentage)
      (AddonDeco|SetMaxSocketAmount)
      (AddonDeco|GetSocketAmountfromPercentage)
      (AddonDeco|RandomizeSpawnSockets))
    (:then_1
      (bind _returnvalue_2 (AddComponent|NODE_AddHierarchicalInstancedStaticMeshComponent-4))
      (Components|StaticMesh|SetStaticMesh _returnvalue_2 (Variables|HiddenProperties|GetInstancedMeshToSpawn))
      (Utilities|Array|Add (Variables|HiddenProperties|GetAR_InstancedMeshComponents) _returnvalue_2)
      (for _array_element (Variables|HiddenProperties|GetAR_RandomizedSockets)
        (AddonDeco|AddNewInstancetoComponent _returnvalue_2 (Transformation|GetSocketTransform (Variables|Debug|GetStaticMeshComponent) _array_element "RTS_Actor"))
        (if (Variables|AddonDeco|GetEnableCollisionOnDeco)
          (Collision|SetCollisionEnabled _returnvalue "QueryAndPhysics")
          (Components|SetIsReplicated _returnvalue true)))
      (AddonDeco|RandomizeMaterials))))
